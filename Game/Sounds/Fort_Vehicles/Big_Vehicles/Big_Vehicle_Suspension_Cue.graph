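# Sound cue for big vehicle suspension sounds, built from police car suspension stereo waves and classed as an impact sound.
# A quality switch gives the first two levels a random pick of three waves and the last two only the first wave.
# Both branches lower the pitch and boost volume with random modulation, then pass through attenuation nodes.
# The cue overrides attenuation and limits concurrency to two instances, stopping the oldest.
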SoundCue /Game/Sounds/Fort_Vehicles/Big_Vehicles/Big_Vehicle_Suspension_Cue
root=QualityLevel_0
nodes [
 {id=Attenuation_0, children [Modulator_0], properties {bOverrideAttenuation=true, attenuationOverrides {nonSpatializedRadiusStart=200}}}
 {id=Attenuation_2, children [Modulator_2], properties {bOverrideAttenuation=true, attenuationOverrides {nonSpatializedRadiusStart=200}}}
 {id=Modulator_0, children [WavePlayer_0], properties {pitchMax=0.9, pitchMin=0.7, volumeMax=1.5, volumeMin=1.3}}
 {id=Modulator_2, children [Random_1], properties {pitchMax=0.9, pitchMin=0.7, volumeMax=1.5, volumeMin=1.3}}
 {id=QualityLevel_0, children [Attenuation_2, Attenuation_2, Attenuation_0, Attenuation_0]}
 {id=Random_1, children [WavePlayer_4,WavePlayer_5,WavePlayer_6], properties {weights [1,1,1]}}
 {id=WavePlayer_0, properties {soundWaveAssetPtr=/Game/Sounds/Fort_Vehicles/Police/Police_Car_Suspension_Stereo_01}}
 {id=WavePlayer_4, properties {soundWaveAssetPtr=&Police_Car_Suspension_Stereo_01}}
 {id=WavePlayer_5, properties {soundWaveAssetPtr=/Game/Sounds/Fort_Vehicles/Police/Police_Car_Suspension_Stereo_02}}
 {id=WavePlayer_6, properties {soundWaveAssetPtr=/Game/Sounds/Fort_Vehicles/Police/Police_Car_Suspension_Stereo_03}}
]
properties {
 attenuationSettings=/Game/Sounds/Attenuation/Fort_Prop_Vehicles_Att
 bOverrideAttenuation=true
 bOverrideConcurrency=true
 maxDistance=4000
 soundClassObject=&Fort_Impacts
 volumeMultiplier=1.3
 attenuationOverrides {nonSpatializedRadiusStart=500}
 concurrencyOverrides {maxCount=2, resolutionRule=StopOldest}
}
